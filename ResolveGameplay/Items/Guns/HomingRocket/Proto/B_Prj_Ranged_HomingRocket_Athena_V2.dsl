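; Homing rocket projectile child Blueprint that extends Javelin homing and lock-on warnings to support player pawns as targets.
; A looping timer caches the target pawn's location and checks winged flight, with a downward trace as rough low-altitude test, to drop tracking.
; Losing the pawn removes a lock-on stack and the incoming-missile effect, then homes on its last location.
; Warning widget effects go to targeted pawns or vehicle passengers, with enter/exit events bound.
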(event EventUpdateJavelinHomingData
  (params (LockOnTarget :type Actor%20Object%20Reference) (TargetLoc :type Vector :default "0, 0, 0"))
  ;[0.2 0.2 1] Adding in support for pawn targeting
  ;;[0 0 0] UpdateJavelinHomingData Functionality Extension - If target is pawn cache the pawn and start tracking data and homing fail conditions
  (|CallParentFunction LockOnTarget TargetLoc)
  (bind _asfort_player_pawn (Utilities|Casting|CastToFortPlayerPawn LockOnTarget))
  (Variables|Default|SetTargetAsPawn _asfort_player_pawn)
  (CallFunction|UpdatePawnTargetData)
  (bind _returnvalue (Utilities|Time|SetTimerbyEvent (EventDispatchers|CreateEvent :!event Custom|UpdatePawnTargetData) 0.1 true true))
  (Variables|Default|SetPawnUpdateTimer _returnvalue))

(event Custom|UpdatePawnTargetData
  ;[0.2 0.2 1] Adding in support for pawn targeting
  ;;[0 0 0] Pawn information we want to track and conditions that may trigger the rocket to lose the pawn as a target
  (bind _targetaspawn (Variables|Default|GetTargetAsPawn)
    (:then
      (Variables|Default|SetTarget%20Loc (Transformation|GetActorLocation _targetaspawn)) ; Cache location
      (bind _returnvalue (GameplayTags|HasMatchingGameplayTag(Message) _targetaspawn "(TagName=\"MovementMode.Flying.WingedFlight\")"))
      (if _returnvalue
        ;;;[1 1 0] TBD - Lose target when low altitude.
        ;;; Needs improvement as it is just checking for visibility channel blocking hits
        (bind _outhit (Collision|LineTraceByChannel (Transformation|GetActorLocation _targetaspawn) (+ (Transformation|GetActorLocation _targetaspawn) "0.000000,0.000000,-250.000000") "TraceTypeQuery1" false (Utilities|Array|MakeArray _targetaspawn)))
        (if (Collision|BreakHitResult _outhit)
          _)
        (else
          _))
      ;;;[1 0.25 0] Lost Pawn as Homing Target
      (CallFunction|HandleLostPawnasTarget))
    (:else
      _)))

(event Custom|HandleLostPawnAsTarget
  ;[0.2 0.2 1] Adding in support for pawn targeting
  ;;[0 0 0] LostPawnAsTarget - Clear timer, set javelin homing data to last known player position and clear TargetAsPawn ref
  (CallFunction|RemoveOneLockonStack)
  (Utilities|Time|ClearandInvalidateTimerbyHandle (Variables|Default|GetPawnUpdateTimer))
  (CallFunction|UpdateJavelinHomingData 0 (Variables|Default|GetTarget%20Loc))
  (Development|PrintString "Lost tracking on pawn" true true "(R=1.000000,G=0.000000,B=0.000000,A=1.000000)" 2e+01)
  ;;;[1 1 1]
  (bind _targetaspawn (Variables|Default|GetTargetAsPawn)
    (:then
      (Utilities|IsValid (Ability|GetAbilitySystemComponent _targetaspawn)
        (:Is%20Valid
          (GameplayEffects|RemoveActiveGameplayEffectbySourceEffect (Ability|GetAbilitySystemComponent _targetaspawn) (Variables|Default|GetGEIncomingMissile) (Ability|GetAbilitySystemComponent _targetaspawn) -1))))
    (:else)))

(event EventAddWarningWidgetToTargets
  (params (Target :type Object%20Reference))
  ;[0.2 0.2 1] Adding in support for pawn targeting
  ;;[0 0 0] AddWarningWidgetToTargets Override - If target is vehicle add warning widget to passengers, if target is a pawn add warning widget to that pawn
  (bind _returnvalue (GameplayTags|HasMatchingGameplayTag(Message) Target "(TagName=\"Vehicle\")"))
  (if _returnvalue
    ;;;[0.5176 0.7606 1] If vehicle bind to vehicles player enter/exit events
    (bind _asfort_athena_vehicle (Utilities|Casting|CastToFortAthenaVehicle Target))
    (Vehicle|BindEventtoOnPawnFinishEnteringVehicle _asfort_athena_vehicle (EventDispatchers|CreateEvent :!event PlayerEnteredVehicle))
    (Vehicle|BindEventtoOnPawnExitedVehicleSeat _asfort_athena_vehicle (EventDispatchers|CreateEvent :!event PlayerExitedVehicle))
    ;;;[0.5176 0.7606 1] Get all current passengers and create an array, loop through and apply the widget GE to each
    (bind _returnvalue_2 (Vehicle|GetAllPassengers(Message) Target))
    (bind _output_get (Variables|Default|SetCurrentVehiclePassengers _returnvalue_2))
    (for _array_element _output_get
      (CallFunction|ApplyLockonWidgetGE _array_element))
    (else
      ;;;[0.1723 0.3144 1] If target is a pawn just add the lock on widget
      (bind _asfort_player_pawn (Utilities|Casting|CastToFortPlayerPawn Target))
      (CallFunction|ApplyLockonWidgetGE _asfort_player_pawn))))

(event EventOnExploded
  (params (HitActors :type Array%20of%20Actor%20Object%20References) (HitResults :type Array%20of%20Hit%20Result%20Structures))
  ;[0.2 0.2 1] Adding in support for pawn targeting
  ;;[0 0 0] OnExploded Functionality Extension - Clear Pawn Timer
  (|CallParentFunction HitActors HitResults)
  (Utilities|Time|ClearandInvalidateTimerbyHandle (Variables|Default|GetPawnUpdateTimer)))

(fn RemoveOneLockOnStack
  (|CallParentFunction)
  (bind _targetaspawn (Variables|Default|GetTargetAsPawn)
    (:then
      (Ability|SendGameplayEventtoActor _targetaspawn "(TagName=\"Event.Resolve.Item.Weapon.HomingRocket.RemoveOneLockOnStack\")"))
    (:else)))

(fn ConstructionScript
  (|CallParentFunction))
