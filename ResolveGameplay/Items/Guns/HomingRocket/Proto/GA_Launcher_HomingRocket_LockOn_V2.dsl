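; Gameplay ability logic for the homing rocket launcher's lock-on, finding and validating targets the player can lock onto.
; ExecuteTraceForLockableTargets runs a multi-sphere trace from the targeting camera, using scalable floats for offset, angle and distance, with optional debug drawing.
; It takes the first vehicle-tagged or flying-pawn-tagged hit, falling back to the actor's owner for wing actors attached to pawns.
; IsPotentialTargetValid checks the stored potential target against the valid target tags.
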(fn ExecuteTraceForLockableTargets
  (params (TracePosition :type Integer))
  (bind (_returnvalue_location _returnvalue_rotation) (Math|Transform|BreakTransform (Targeting|GetTargetingTransform (Variables|Default|GetPlayerRef) "Camera" 0 :!class /Script/FortniteGame.FortPawn)))
  (bind _outhits (Collision|MultiSphereTraceByChannel (+ _returnvalue_location (* (Math|Vector|Normalize (+ (* (Math|Vector|GetUpVector _returnvalue_rotation) 0) (* 0 (Math|Vector|GetRightVector _returnvalue_rotation)))) (ScalableFloat|GetValueatLevel (Variables|Anvil|ScalableFloats|GetTargetingOffsetSizeMult)))) (+ (+ _returnvalue_location (* (Math|Vector|Normalize (+ (* (Math|Vector|GetUpVector _returnvalue_rotation) 0) (* 0 (Math|Vector|GetRightVector _returnvalue_rotation)))) (ScalableFloat|GetValueatLevel (Variables|Anvil|ScalableFloats|GetTargetingOffsetSizeMult)))) (* (Math|Vector|Lerp(Vector) (Math|Vector|Normalize (+ (* (Math|Vector|GetUpVector _returnvalue_rotation) 0) (* 0 (Math|Vector|GetRightVector _returnvalue_rotation)))) (Math|Vector|GetForwardVector _returnvalue_rotation) (ScalableFloat|GetValueatLevel (Variables|Anvil|ScalableFloats|GetTraceAngle))) (ScalableFloat|GetValueatLevel (Variables|Anvil|ScalableFloats|GetTraceDistance)))) 4e+01 "TraceTypeQuery6" false (Utilities|Array|MakeArray (Variables|Default|GetPlayerRef)) (select (Variables|Anvil|Booleans|GetDebug) "ForDuration" "None") true "(R=1.000000,G=0.000000,B=0.000000,A=1.000000)" "(R=0.000000,G=1.000000,B=0.000000,A=1.000000)" 0.5))
  (bind _array_element (Utilities|Array|ForEachLoopwithBreak _outhits :!label _label)
    (:LoopBody
      ;[1 1 1] Stretch Goal: Gameplay Tag(s) to allow creative to target other things
      (bind _returnvalue (GameplayTags|HasMatchingGameplayTag(Message) (.HitActor (Collision|BreakHitResult _array_element)) (Variables|Default|GetVehicleTag)))
      (if _returnvalue
        (LocalVariables|SetValidTargetActor (.HitActor (Collision|BreakHitResult _array_element)))
        (goto _label Break)
        (else
          (bind _returnvalue_2 (GameplayTags|HasMatchingGameplayTag(Message) (.HitActor (Collision|BreakHitResult _array_element)) (Variables|Default|GetFlyingPawnTag)))
          (if _returnvalue_2
            (bind _asfort_player_pawn (Utilities|Casting|CastToFortPlayerPawn (.HitActor (Collision|BreakHitResult _array_element)))
              (:then
                (LocalVariables|SetValidTargetActor _asfort_player_pawn)
                (goto _label Break))
              (:CastFailed
                (Utilities|IsValid (Actor|GetOwner (.HitActor (Collision|BreakHitResult _array_element)))
                  (:Is%20Valid
                    (LocalVariables|SetValidTargetActor (Actor|GetOwner (.HitActor (Collision|BreakHitResult _array_element))))
                    (goto _label Break))) ; case to handle winged actors attached to player pawn
              )))))
      (goto _label))
    (:Completed
      (return (Utilities|IsValid (LocalVariables|GetValidTargetActor) :!class /Script/Engine.KismetSystemLibrary) true true (LocalVariables|GetValidTargetActor)))))

(fn IsPotentialTargetValid
  (bind _potentialtarget (Variables|Default|GetPotentialTarget)
    (:then
      (bind _returnvalue (GameplayTags|HasAnyMatchingGameplayTags(Message) _potentialtarget (Variables|Default|GetValidTargetTags)))
      (return _returnvalue))
    (:else
      (return false))))
